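# Animation Blueprint for the SquidGlisten Lift male medium body cosmetic, which follows the attached parent mesh's pose and layers a Control Rig on top.
# Copy Pose From Mesh feeds the body's Control Rig, which also receives a grind-rail flag, a smoothed emote float and a running-or-emote flag.
# Each update, the float interpolates at constant speed toward emoting or skydiving, and the flag is set when emoting, sprinting or moving while not crouching.
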AnimBlueprint /BRCosmetics/Characters/Player/Male/Medium/Bodies/M_MED_SquidGlisten_Lift/Meshes/M_MED_SquidGlisten_Lift_AnimBP
target_skeleton=&/M_MED_SquidGlisten_Lift_Skeleton
graphs [
 AnimGraph {nodes [
   n1 = Control Rig!(
    Source=Copy Pose From Mesh() {bUseAttachedParent=true}
    Alpha
    Emote Float:double=Get is Emote Float() -> (is Emote Float:double)
    on GrindRail:bool=Get bIsOnGrindRail()
    is RunningOrEmote:bool=Get isRunningOrEmote() -> (isRunningOrEmote:bool)
   ) -> (Pose) {
    alphaBoolBlend {blendOption=Linear}
    controlRigAssetReference {blueprintRigClass=&/M_MED_SquidGlisten_Lift_CtrlRig.M_MED_SquidGlisten_Lift_CtrlRig_C}
   }
   Output Pose(Result=@n1) {layerGroup=None}
  ]}
 EventGraph {nodes [
   n1 = Event BlueprintUpdateAnimation()
   > Set Delta Time X(Delta Time X:double=@n1.DeltaTimeX) -> (Output_Get:double)
   n2 = OR Boolean(A=Get bIsPlayingEmote(), B=Get bIsSkyDiving())
   > Set is Emote Float(is Emote Float:double=FInterpTo_Constant!(
     Current:double=Get is Emote Float()
     Target:double="To Float (Boolean)"(InBool=@n2)
     DeltaTime:double=Get Delta Time X() -> (Delta Time X:double)
     InterpSpeed:double=4
    ) -> (ReturnValue:double)) -> (Output_Get:double)
   n3 = Get bIsCrouching()
   n4 = Get bIsPlayingEmote()
   n5 = Get bIsSprinting()
   n6 = Get bIsMoving2D()
   > Set isRunningOrEmote(isRunningOrEmote:bool=AND Boolean(A=OR Boolean!(A=@n4, B=@n5, C:bool=@n6) -> (ReturnValue), B=NOT Boolean(A=@n3))) -> (Output_Get:bool)
  ]}
]
variables [
 Delta Time X {type=number}
 is Emote Float {type=number}
 isRunningOrEmote {type=boolean}
]
